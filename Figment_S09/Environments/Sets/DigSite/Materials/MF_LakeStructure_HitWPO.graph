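# World position offset function for hit reactions on DigSite lake structures, plus a wobble Mask output.
# Time since GameTime, normalized by WobbleDuration and eased, grows a sphere mask from WSDeformationCenter, sized by object radius.
# That mask drives a fading, vibrating displacement along vertex normals, scaled by WobbleScale.
# A fixed sphere mask adds a push along Code_PrecomputedHitVector, which this function does not fade over time.
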MaterialFunction /Figment_S09/Environments/Sets/DigSite/Materials/MF_LakeStructure_HitWPO
graph {nodes [
  n1 = Multiply(A=VectorParameter() {group=z_code, parameterName=Code_PrecomputedHitVector, defaultValue=<rgba 0 0 1 0>}.RGB) {constB=0.25}
  n2 = WorldPosition()
  n3 = VectorParameter() {group=z_code, parameterName=WSDeformationCenter, defaultValue=<rgba 0 50 0 0>}
  n4 = ObjectRadius()
  n5 = SphereMask(A=@n2.XYZ, B=@n3.RGB, Radius=Multiply(A=@n4) {constB=0.4}) {hardnessPercent=0}
  n6 = Subtract(A=Time(), B=ScalarParameter() {parameterName=GameTime})
  n7 = FunctionInput(Preview=ScalarParameter() {defaultValue=0.5, parameterName=WobbleDuration}) {bUsePreviewValueAsDefault=true, inputName=WobbleDuration, inputType=FunctionInput_Scalar}
  n8 = Power(Base=Clamp(None=Divide(A=@n6, B=@n7))) {constExponent=0.6}
  n9 = SphereMask(A=@n2.XYZ, B=@n3.RGB, Radius=Multiply(A=Multiply(A=@n4) {constB=1.25}, B=@n8)) {hardnessPercent=0}
  n10 = Add(A=@n9, B=Sine(None=@n9))
  n11 = OneMinus(None=@n8)
  n12 = ScalarParameter() {defaultValue=0.4, parameterName=WobbleVibrationAmount}
  n13 = Multiply(
   A=@n10
   B=Multiply(
    A=Sine(None=@n11) {desc=Vibration, period=0.05}
    B=FunctionInput(Preview=@n12) {bUsePreviewValueAsDefault=true, inputName=VibrationAmount, inputType=FunctionInput_Scalar}
   ) {constB=0.25}
  )
  n14 = Multiply(A=VertexNormalWS(), B=Multiply(A=Add(A=@n10, B=@n13), B=@n11))
  n15 = FunctionInput(Preview=ScalarParameter() {defaultValue=10, parameterName=WobbleScale}) {bUsePreviewValueAsDefault=true, inputName=WobbleScale, inputType=FunctionInput_Scalar}
  FunctionOutput(None=Add(A=Multiply(A=@n1, B=@n5) {constA=10}, B=Multiply(A=@n14, B=@n15)))
  FunctionOutput(None=Saturate(None=Multiply(A=@n10, B=@n11))) {outputName=Mask, sortPriority=1}
  Time() {bOverride_Period=true, period=1}
 ]}
