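# Base master material for DigSite lake structures, blending four tintable texture layers with optional parallax and DM color masks.
# Adds view-based chromatic sheen, distance-faded detail normal and roughness, flickering emissive, then gradient, vertex-paint and up-vector texture functions.
# MF_LakeStructure_HitWPO supplies an offset and a mask that flattens normals and overrides roughness and base color.
# The final world position offset is replaced by the DanceClub switch: dance-club motion or zero.
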Material /Figment_S09/Environments/Sets/DigSite/Materials/M_BaseLakeStructure
graph {
 properties {
  bUseMaterialAttributes=true
  ditheredLODTransition=true
  physMaterial=/Game/FortressPhysicalMaterials/Metal
  pixelDepthOffsetMode=PDOM_Legacy
  usedShadingModels=Default Lit
 }
 nodes [
  n1 = MaterialFunctionCall() {materialFunction=/Game/Athena/WinterArt/Materials/MF_UVsSet}
  n2 = TextureObjectParameter() {group=Material_1, parameterName=Normals, samplerType=SAMPLERTYPE_Normal, sortPriority=2, texture=/Engine/EngineMaterials/FlatNormal}
  n3 = ScalarParameter() {defaultValue=1, group=Material_1, parameterName=Mat1 UV Scale, sortPriority=3}
  n4 = Multiply(A=TextureCoordinate(), B=Clamp(None=Ceil(None=@n3)) {maxDefault=1000, minDefault=1})
  n5 = TextureSample!(UVs=@n1.POM UVs 4 Layers, Tex=@n2, "DDX(UVs)"=DDX(Value=@n4), "DDY(UVs)"=DDY(Value=@n4), Apply View MipBias) {mipValueMode=TMVM_Derivative, samplerType=SAMPLERTYPE_Normal}
  n6 = TextureSample!(UVs=@n1.UVs Layer 1, Tex=@n2, Apply View MipBias) {samplerType=SAMPLERTYPE_Normal}
  n7 = QualitySwitch(
   Default=@n1.UVs Layer 1
   High=StaticSwitchParameter(True=@n1.POM UVs 4 Layers, False=@n1.POM UVs 4 Layers) {group=Material_1, parameterName=POM_True, sortPriority=11}
  )
  n8 = TextureObjectParameter() {
   group=Material_1
   parameterName=SpecularMasks
   samplerType=SAMPLERTYPE_Masks
   sortPriority=1
   texture=/Game/Packages/Fortnite_UniqueMaterials_1/Textures/T_Fortnite_Default_S
  }
  n9 = TextureSample(UVs=@n7, Tex=@n8) {samplerType=SAMPLERTYPE_Masks, texture=/Game/Environments/Sets/IceCastle/Textures/T_IceCastle_Wall1_D}
  n10 = MaterialFunctionCall() {materialFunction=&MF_UVsSet}
  n11 = QualitySwitch(
   Default=@n10.UVs Layer 3
   High=StaticSwitchParameter(True=@n10.POM UVs 4 Layers, False=@n10.POM UVs 4 Layers) {group=Material_3, parameterName=L3_POM_True, sortPriority=11}
  )
  n12 = TextureSampleParameter2D(UVs=@n11) {
   desc=Use for trims and accents
   group=Material_3
   parameterName=SpecularMasks_3
   samplerSource=SSM_Wrap_WorldGroupSettings
   samplerType=SAMPLERTYPE_Masks
   sortPriority=2
   texture=&T_Fortnite_Default_S
  }
  n13 = TextureSampleParameter2D(UVs=@n10.UVs Layer 3) {
   desc=Use for trims and accents
   group=Material_3
   parameterName=Normals_Texture_3
   samplerSource=SSM_Wrap_WorldGroupSettings
   samplerType=SAMPLERTYPE_Normal
   sortPriority=3
   texture=&FlatNormal
  }
  n14 = TextureSampleParameter2D!(UVs=@n10.POM UVs 4 Layers, "DDX(UVs)"=DDX(Value=@n10.UVs Layer 3), "DDY(UVs)"=DDY(Value=@n10.UVs Layer 3), Apply View MipBias) {
   desc=Use for trims and accents
   group=Material_3
   mipValueMode=TMVM_Derivative
   parameterName=Normals_Texture_3
   samplerSource=SSM_Wrap_WorldGroupSettings
   samplerType=SAMPLERTYPE_Normal
   sortPriority=3
   texture=&FlatNormal
  }
  n15 = MaterialFunctionCall() {materialFunction=&MF_UVsSet}
  n16 = QualitySwitch(
   Default=@n15.UVs Layer 4
   High=StaticSwitchParameter(True=@n15.POM UVs 4 Layers, False=@n15.POM UVs 4 Layers) {group=Material_4, parameterName=L4_POM_True, sortPriority=11}
  )
  n17 = TextureSampleParameter2D!(UVs=@n16, Apply View MipBias) {
   desc=Use for doors
   group=Material_4
   parameterName=SpecularMasks_4
   samplerSource=SSM_Wrap_WorldGroupSettings
   samplerType=SAMPLERTYPE_Masks
   sortPriority=2
   texture=&T_Fortnite_Default_S
  }
  n18 = TextureSampleParameter2D() {
   desc=Use for doors
   group=Material_4
   parameterName=Normals_Texture_4
   samplerSource=SSM_Wrap_WorldGroupSettings
   samplerType=SAMPLERTYPE_Normal
   sortPriority=3
   texture=&FlatNormal
  }
  n19 = TextureSampleParameter2D!(UVs=@n15.POM UVs 4 Layers, "DDX(UVs)"=DDX(Value=@n15.UVs Layer 4), "DDY(UVs)"=DDY(Value=@n15.UVs Layer 4), Apply View MipBias) {
   desc=Use for doors
   group=Material_4
   mipValueMode=TMVM_Derivative
   parameterName=Normals_Texture_4
   samplerSource=SSM_Wrap_WorldGroupSettings
   samplerType=SAMPLERTYPE_Normal
   sortPriority=3
   texture=&FlatNormal
  }
  n20 = TextureObjectParameter() {
   group=Material_1
   parameterName=DM
   samplerType=SAMPLERTYPE_Masks
   sortPriority=1
   texture=&T_Fortnite_Default_S
  }
  n21 = TextureSample(UVs=@n7, Tex=@n20) {samplerType=SAMPLERTYPE_Masks}
  n22 = VectorParameter() {group=Material_1, parameterName=TintColor, sortPriority=7, defaultValue=<rgba 1 1 1 0>}
  n23 = MaterialFunctionCall!(
   DM=@n21.RGB
   Color1=@n22.RGB
   Color2=VectorParameter() {group=Material_1, parameterName=M1_TintColor2, sortPriority=5, defaultValue=<rgba 1 1 1 0>}.RGB
   Color3=VectorParameter() {group=Material_1, parameterName=M1_TintColor3, sortPriority=5, defaultValue=<rgba 1 1 1 0>}.RGB
  ) {materialFunction=/Game/Athena/WinterArt/Materials/MF_DMTexture}
  n24 = TextureObjectParameter() {group=Material_1, parameterName=Diffuse, texture=&T_IceCastle_Wall1_D}
  n25 = TextureSample(UVs=@n7, Tex=@n24) {texture=&T_IceCastle_Wall1_D}
  n26 = MaterialFunctionCall!(
   Floor1=@n22.RGB
   Floor2=VectorParameter() {group=Material_1, parameterName=Floor2, sortPriority=8, defaultValue=<rgba 1 1 1 0>}.RGB
   Floor3=VectorParameter() {group=Material_1, parameterName=Floor3, sortPriority=9, defaultValue=<rgba 1 1 1 0>}.RGB
   Basement=VectorParameter() {group=Material_1, parameterName=Basement, sortPriority=10, defaultValue=<rgba 1 1 1 0>}.RGB
  ) {materialFunction=/Game/Athena/WinterArt/Materials/MF_WS_GridUV}
  n27 = Multiply(
   A=@n25.RGB
   B=MaterialFunctionCall!(In=StaticSwitchParameter(True=@n26.StoryMaskOut, False=@n22.RGB) {group=Material_1, parameterName=UsePerFloorColor, sortPriority=4}) {materialFunction=/Game/Athena/WinterArt/Materials/MF_VPTintMask}.Result
  )
  n28 = MaterialFunctionCall() {materialFunction=&MF_UVsSet}
  n29 = QualitySwitch(
   Default=@n28.UVs Layer 2
   High=StaticSwitchParameter(True=@n28.POM UVs 4 Layers, False=@n28.POM UVs 4 Layers) {group=Material_2, parameterName=L2_POM_True, sortPriority=11}
  )
  n30 = TextureObjectParameter() {
   group=Material_2
   parameterName=DM_2
   samplerType=SAMPLERTYPE_Masks
   sortPriority=1
   texture=&T_Fortnite_Default_S
  }
  n31 = TextureSample(UVs=@n29, Tex=@n30) {samplerType=SAMPLERTYPE_Masks}
  n32 = VectorParameter() {group=Material_2, parameterName=M2_TintColor, sortPriority=5, defaultValue=<rgba 1 1 1 0>}
  n33 = MaterialFunctionCall(
   DM=@n31.RGB
   Color1=@n32.RGB
   Color2=VectorParameter() {group=Material_2, parameterName=M2_TintColor2, sortPriority=5, defaultValue=<rgba 1 1 1 0>}.RGB
   Color3=VectorParameter() {group=Material_2, parameterName=M2_TintColor3, sortPriority=5, defaultValue=<rgba 1 1 1 0>}.RGB
  ) {materialFunction=&MF_DMTexture}
  n34 = TextureSampleParameter2D!(UVs=@n29, Apply View MipBias) {
   desc=Use for interior walls and ceilings
   group=Material_2
   parameterName=Diffuse_Texture_2
   samplerSource=SSM_Wrap_WorldGroupSettings
   sortPriority=1
   texture=/Game/Environments/Sets/Dojo/HomeBase/Textures/T_midgrey
  }
  n35 = Multiply(A=MaterialFunctionCall(In=@n32.RGB) {materialFunction=&MF_VPTintMask}.Result, B=@n34.RGB)
  n36 = TextureSampleParameter2D(UVs=@n28.UVs Layer 2) {
   desc=Use for interior walls and ceilings
   group=Material_2
   parameterName=Normals_Texture_2
   samplerSource=SSM_Wrap_WorldGroupSettings
   samplerType=SAMPLERTYPE_Normal
   sortPriority=3
   texture=&FlatNormal
  }
  n37 = TextureSampleParameter2D!(UVs=@n28.POM UVs 4 Layers, "DDX(UVs)"=DDX(Value=@n28.UVs Layer 2), "DDY(UVs)"=DDY(Value=@n28.UVs Layer 2), Apply View MipBias) {
   desc=Use for interior walls and ceilings
   group=Material_2
   mipValueMode=TMVM_Derivative
   parameterName=Normals_Texture_2
   samplerSource=SSM_Wrap_WorldGroupSettings
   samplerType=SAMPLERTYPE_Normal
   sortPriority=3
   texture=&FlatNormal
  }
  n38 = TextureSampleParameter2D!(UVs=@n29, Apply View MipBias) {
   desc=Use for interior walls and ceilings
   group=Material_2
   parameterName=SpecularMasks_2
   samplerSource=SSM_Wrap_WorldGroupSettings
   samplerType=SAMPLERTYPE_Masks
   sortPriority=2
   texture=&T_Fortnite_Default_S
  }
  n39 = TextureObjectParameter() {
   group=Material_3
   parameterName=DM_3
   samplerType=SAMPLERTYPE_Masks
   sortPriority=1
   texture=&T_Fortnite_Default_S
  }
  n40 = TextureSample(Tex=@n39) {samplerType=SAMPLERTYPE_Masks}
  n41 = VectorParameter() {group=Material_3, parameterName=M3_TintColor, sortPriority=5, defaultValue=<rgba 1 1 1 0>}
  n42 = MaterialFunctionCall(
   DM=@n40.RGB
   Color1=@n41.RGB
   Color2=VectorParameter() {group=Material_3, parameterName=M3_TintColor2, sortPriority=5, defaultValue=<rgba 1 1 1 0>}.RGB
   Color3=VectorParameter() {group=Material_3, parameterName=M3_TintColor3, sortPriority=5, defaultValue=<rgba 1 1 1 0>}.RGB
  ) {materialFunction=&MF_DMTexture}
  n43 = TextureSampleParameter2D(UVs=@n11) {
   desc=Use for trims and accents
   group=Material_3
   parameterName=Diffuse_Texture_3
   samplerSource=SSM_Wrap_WorldGroupSettings
   sortPriority=1
   texture=&T_midgrey
  }
  n44 = Multiply(A=MaterialFunctionCall(In=@n41.RGB) {materialFunction=&MF_VPTintMask}.Result, B=@n43.RGB)
  n45 = TextureObjectParameter() {
   group=Material_4
   parameterName=DM_4
   samplerType=SAMPLERTYPE_Masks
   sortPriority=1
   texture=&T_Fortnite_Default_S
  }
  n46 = TextureSample(Tex=@n45) {samplerType=SAMPLERTYPE_Masks}
  n47 = VectorParameter() {group=Material_4, parameterName=M4_TintColor, sortPriority=5, defaultValue=<rgba 1 1 1 0>}
  n48 = MaterialFunctionCall(
   DM=@n46.RGB
   Color1=@n47.RGB
   Color2=VectorParameter() {group=Material_4, parameterName=M4_TintColor2, sortPriority=5, defaultValue=<rgba 1 1 1 0>}.RGB
   Color3=VectorParameter() {group=Material_4, parameterName=M4_TintColor3, sortPriority=5, defaultValue=<rgba 1 1 1 0>}.RGB
  ) {materialFunction=&MF_DMTexture}
  n49 = TextureSampleParameter2D(UVs=@n16) {
   desc=Use for doors
   group=Material_4
   parameterName=Diffuse_Texture_4
   samplerSource=SSM_Wrap_WorldGroupSettings
   sortPriority=1
   texture=&T_midgrey
  }
  n50 = Multiply(A=MaterialFunctionCall(In=@n47.RGB) {materialFunction=&MF_VPTintMask}.Result, B=@n49.RGB)
  n51 = MaterialFunctionCall!(
   D=StaticSwitchParameter(True=AppendVector(A=@n23.Result, B=@n21.A), False=AppendVector(A=@n27, B=@n25.A)) {desc=Creative, group=Material_1, parameterName=LM_DMTextureTrue, sortPriority=11}
   S=AppendVector(A=@n9.RGB, B=@n9.RGB)
   N=QualitySwitch(Default=@n6.RGB, High=StaticSwitchParameter(True=@n5.RGB, False=@n6.RGB) {group=Material_1, parameterName=POM_True, sortPriority=11})
   D2=StaticSwitchParameter(True=AppendVector(A=@n33.Result, B=@n31.A), False=AppendVector(A=@n35, B=@n34.A)) {desc=Creative, group=Material_2, parameterName=L2_DMTextureTrue, sortPriority=11}
   S2=AppendVector(A=@n38.RGB, B=@n38.RGB)
   N2=QualitySwitch(Default=@n36.RGB, High=StaticSwitchParameter(True=@n37.RGB, False=@n36.RGB) {group=Material_2, parameterName=L2_POM_True, sortPriority=11})
   D3=StaticSwitchParameter(True=AppendVector(A=@n42.Result, B=@n40.A), False=AppendVector(A=@n44, B=@n43.A)) {desc=Creative, group=Material_3, parameterName=L3_DMTextureTrue, sortPriority=11}
   S3=AppendVector(A=@n12.RGB, B=@n12.RGB)
   N3=QualitySwitch(Default=@n13.RGB, High=StaticSwitchParameter(True=@n14.RGB, False=@n13.RGB) {group=Material_3, parameterName=L3_POM_True, sortPriority=11})
   D4=StaticSwitchParameter(True=AppendVector(A=@n48.Result, B=@n46.A), False=AppendVector(A=@n50, B=@n49.A)) {desc=Creative, group=Material_3, parameterName=L3_DMTextureTrue, sortPriority=11}
   S4=AppendVector(A=@n17.RGB, B=@n17.RGB)
   N4=QualitySwitch(Default=@n18.RGB, High=StaticSwitchParameter(True=@n19.RGB, False=@n18.RGB) {group=Material_4, parameterName=L4_POM_True, sortPriority=11})
   E1=VectorParameter() {group=Material_1, parameterName=Emissive, sortPriority=7}.RGB
   E2=VectorParameter() {group=Material_2, parameterName=Emissive2, sortPriority=7}.RGB
   E3=VectorParameter() {group=Material_3, parameterName=Emissive3, sortPriority=7}.RGB
   E4=VectorParameter() {group=Material_4, parameterName=Emissive4, sortPriority=7}.RGB
  ) {materialFunction=/FigmentEnvironment/LM_8_1_1/Athena/WinterArt/Materials/MF_FN_TextureSetLerp_4Layers}
  n52 = ScalarParameter() {defaultValue=5, group=Global, parameterName=SpecFresnelExp}
  n53 = VertexNormalWS()
  n54 = LinearInterpolate(
   A=Multiply(A=ScalarParameter() {group=Global, parameterName=Spec Min Fresnel}, B=@n51.S)
   B=@n51.S
   Alpha=OneMinus(None=Fresnel(ExponentIn=@n52, Normal=@n53) {exponent=3})
  )
  LinearInterpolate(B=@n54)
  n55 = ParticleRandom()
  n56 = DynamicParameter() {paramNames [FlickerFade,FlickerSpeed,Param3,Param4]}
  n57 = ScalarParameter() {defaultValue=0.2, group=Emissive, parameterName=Emissive Flicker Speed}
  n58 = Multiply(
   A=Time()
   B=StaticSwitchParameter(True=LinearInterpolate(Alpha=@n55) {constA=1.2, constB=0.8}, False=Multiply(A=@n56.FlickerSpeed, B=@n57)) {parameterName=GPU Particles}
  )
  n59 = Sine(None=@n58) {period=0.075}
  n60 = Sine(None=@n58) {period=0.155}
  n61 = Sine(None=@n58) {period=0.2}
  Multiply(A=@n59, B=Multiply(A=Multiply(A=@n60, B=@n61), B=Sine(None=@n58) {period=0.43}))
  n62 = Transform(None=PixelNormalWS()) {transformSourceType=TRANSFORMSOURCE_World, transformType=TRANSFORM_View}
  n63 = Multiply(
   A=Fresnel(ExponentIn=ScalarParameter() {defaultValue=1.0389, group=Chromatic, parameterName=CromaticExponent}) {baseReflectFraction=0}
   B=ScalarParameter() {defaultValue=1.25, group=Chromatic, parameterName=ChromaticFresnelX}
  )
  n64 = ScalarParameter() {defaultValue=-0.8, group=Chromatic, parameterName=ChromaticFresnelAdd}
  n65 = ScalarParameter() {defaultValue=0.1, group=Chromatic, parameterName=ChromaticStaticAmount}
  n66 = TextureSample(UVs=Multiply(A=TextureCoordinate(), B=ScalarParameter() {defaultValue=2, group=Chromatic, parameterName=ChromaticStaticSize})) {texture=/Game/Effects/Fort_Effects/Textures/Tile/T_Inky_Smoke_Tile_Chromatic}
  n67 = AppendVector(
   A=Multiply(A=ComponentMask(None=@n62) {r=true}, B=ScalarParameter() {defaultValue=0.4, group=Chromatic, parameterName=CromaticViewSpaceSize})
   B=Add(A=Add(A=@n63, B=@n64), B=Multiply(A=@n65, B=@n66.R))
  )
  n68 = MaterialFunctionCall!(Blend Range=ScalarParameter() {defaultValue=1000, parameterName=DetailFresRange}, Start Offset=Constant()) {materialFunction=/Engine/Functions/Engine_MaterialFunctions03/Particles/Distance_Blend}
  n69 = Multiply(
   A=Multiply(
    A=TextureSample(UVs=@n67) {texture=&T_Inky_Smoke_Tile_Chromatic}.RGB
    B=ScalarParameter() {defaultValue=0.5, group=Chromatic, parameterName=CromaticMasterAmount}
   )
   B=@n68.Result
  )
  n70 = Constant() {r=1}
  n71 = ScalarParameter() {defaultValue=0.7, parameterName=DetailDiffuseDarkness}
  n72 = TextureSample(UVs=Multiply(A=TextureCoordinate(), B=ScalarParameter() {defaultValue=1.5, parameterName=DetailRoughTile})) {samplerType=SAMPLERTYPE_Masks, texture=/Game/Athena/Prototype/Blueprints/Cube/Materials/T_CubeNoise}
  n73 = TextureSample(UVs=Multiply(A=TextureCoordinate(), B=ScalarParameter() {defaultValue=8, parameterName=DetailNormalTile})) {samplerType=SAMPLERTYPE_Normal, texture=/Figment_S09/Athena/Environments/Nexus/Textures/NexFloor_2_N}
  n74 = Constant3Vector() {constant=<rgba 1 1 0 0>}
  n75 = Add(A=@n51.Normals, B=Multiply(A=Multiply(A=@n73.RGB, B=@n74), B=ScalarParameter() {defaultValue=0.6, parameterName=DetailNormalAmount}))
  n76 = LinearInterpolate(
   A=ScalarParameter() {defaultValue=0.6, parameterName=DetailRough-Low}
   B=ScalarParameter() {defaultValue=0.85, parameterName=DetailRough-High}
   Alpha=@n72.R
  )
  n77 = LinearInterpolate(A=ScalarParameter() {parameterName=BaseRoughMin}, B=ScalarParameter() {defaultValue=1, parameterName=BaseRoughMax}, Alpha=@n51.R)
  n78 = StaticComponentMaskParameter(None=TextureSampleParameter2D() {group=Emissive, parameterName=Emissive Mask Texture, texture=&T_BlacksRGB}.RGB) {defaultG=true, group=Emissive, parameterName=Emissive Texture Mask Channel}
  n79 = Multiply(A=@n78, B=Clamp(None=@n59, Min=ScalarParameter() {group=Emissive, parameterName=Emissive Flicker Min Clamp}) {maxDefault=3})
  n80 = ScalarParameter() {group=Emissive, parameterName=Emissive Flicker Lerp}
  n81 = SetMaterialAttributes!(
   MaterialAttributes
   Base Color=Multiply(A=Add(A=@n51.Diffuse, B=@n69), B=LinearInterpolate(A=@n70, B=@n71, Alpha=@n72.R))
   Specular=Multiply(A=@n54, B=ScalarParameter() {defaultValue=1, parameterName=Spec})
   Metallic=LinearInterpolate(B=ScalarParameter() {defaultValue=1, parameterName=MetallicMax}, Alpha=@n51.M)
   Roughness=LinearInterpolate(A=Multiply(A=@n77, B=@n76), B=@n77, Alpha=@n68.Result)
   Emissive Color=Multiply(A=LinearInterpolate(A=@n78, B=@n79, Alpha=@n80), B=@n51.Emissive)
   Normal=Normalize(VectorInput=LinearInterpolate(A=@n75, B=@n51.Normals, Alpha=@n68.Result))
  ) {attributeSetTypes [
    #1
    #2
    #3
    #4
    #5
    #6
   ]}
  Reroute(None=@n81) {desc=ES3_1}
  Reroute(None=@n51.Diffuse)
  n82 = MaterialFunctionCall!(
   In=MaterialFunctionCall!(transitionFactor, MaterialIn=@n81) {materialFunction=/Game/Athena/WinterArt/Materials/MF_WSGradient}.Color*Gradient
   BGMap=@n51.Height
  ) {materialFunction=/FigmentEnvironment/LM_8_1_1/Athena/WinterArt/Materials/MF_VPTexture}
  n83 = MaterialFunctionCall!(In=@n82.Result, BG_HeightMap=@n51.Height, UpVectorTexture=@n82.Result) {materialFunction=/FigmentEnvironment/LM_8_1_1/Athena/WinterArt/Materials/MF_UpVectorTexture}
  n84 = SetMaterialAttributes(Base Color=GetMaterialAttributes(None=@n83.Result) {attributeGetTypes [#1]}.Base Color) {attributeSetTypes [
    #1
    #2
    #3
    #4
    #5
    #6
   ]}
  n85 = MaterialFunctionCall!(VibrationAmount, WobbleScale, WobbleDuration) {materialFunction=&/MF_LakeStructure_HitWPO}
  n86 = MaterialFunctionCall!(
   Material=MaterialFunctionCall!(
    Material=FeatureLevelSwitch(Default=@n83.Result, ES2=@n84, ES3_1=@n83.Result, SM4_REMOVED=@n83.Result, SM5=@n83.Result, SM6=@n83.Result)
    Mask
    WorldPositionOffset=@n85.Result
   ) {materialFunction=/Engine/Functions/MaterialLayerFunctions/MatLayerBlend_OverrideWorldPositionOffset}.Blended Material
   NormalFlatness=@n85.Mask
   Normal=Constant3Vector() {constant=<rgba 0 0 2 0>}
  ) {materialFunction=/Engine/Functions/MaterialLayerFunctions/MatLayerBlend_NormalFlatten}
  n87 = Constant() {r=0.02}
  SetMaterialAttributes!(
   MaterialAttributes=MaterialFunctionCall!(
    Material=MaterialFunctionCall!(Material=@n86.Blended Material, NewRoughness=@n87, Mask=@n85.Mask) {materialFunction=/Engine/Functions/MaterialLayerFunctions/MatLayerBlend_RoughnessOverride}.Blended Material
    NewBaseColor=VectorParameter() {parameterName=FXBaseColor, defaultValue=<rgba 0.1347 0.1324 0.16 1>}.RGB
    Mask=@n85.Mask
   ) {materialFunction=/Engine/Functions/MaterialLayerFunctions/MatLayerBlend_OverrideBaseColor}.Blended Material
   World Position Offset=StaticSwitchParameter(True=MaterialFunctionCall() {materialFunction=/Game/Environments/Sets/DigSite/Materials/MF_Sneezy_DanceClub}.Result, False=Constant()) {parameterName=DanceClub}
  ) {attributeSetTypes [#7]}
  StaticSwitchParameter(True=@n20, False=@n24) {desc=Creative, group=Material_1, parameterName=L1_DMTextureTrue, sortPriority=11}
  n88 = TextureObjectParameter() {group=Material_3, parameterName=Diffuse_Texture_3, sortPriority=1, texture=&T_midgrey}
  StaticSwitchParameter(True=@n39, False=@n88) {desc=Creative, group=Material_3, parameterName=L3_DMTextureTrue, sortPriority=11}
  n89 = TextureObjectParameter() {group=Material_2, parameterName=Diffuse_Texture_2, sortPriority=1, texture=&T_midgrey}
  StaticSwitchParameter(True=@n30, False=@n89) {desc=Creative, group=Material_2, parameterName=L2_DMTextureTrue, sortPriority=11}
  n90 = TextureObjectParameter() {group=Material_4, parameterName=Diffuse_Texture_4, sortPriority=1, texture=&T_midgrey}
  StaticSwitchParameter(True=@n45, False=@n90) {desc=Creative, group=Material_3, parameterName=L3_DMTextureTrue, sortPriority=11}
 ]
}
