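# Additive unlit material for an energy field, apparently around the DigSite snow pedestal, usable on mesh particles.
# Scrolling flow textures distort tiled gradient lines with RGB splitting, blended with a view-based rainbow and shaped by facing angle and camera distance.
# A timestamped hit-wave ring from sphere masks, plus an Opacity-driven vertical wipe, adds tint, distortion and normal-direction vertex offset, zeroed on mobile.
# Static switches add particle color and depth fade.
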Material /Figment_S09/Environments/Sets/DigSite/Materials/M_Snow_Pedestal_Field
graph {
 properties {bUsedWithMeshParticles=true, blendMode=BLEND_Additive, pixelDepthOffsetMode=PDOM_Legacy, shadingModel=MSM_Unlit, usedShadingModels=Unlit}
 nodes [
  n1 = Constant()
  n2 = WorldPosition()
  n3 = ScalarParameter() {defaultValue=500, parameterName=WPOWipeDist}
  n4 = ScalarParameter() {defaultValue=1, parameterName=Opacity}
  n5 = Add(A=ObjectPositionWS(), B=LinearInterpolate(A=@n3, B=Multiply(A=@n3) {constB=-1}, Alpha=Clamp(None=@n4)))
  n6 = Time()
  n7 = ScalarParameter() {defaultValue=1, desc=Time-stamp, parameterName=GameTime}
  n8 = Saturate(None=Divide(A=Subtract(A=@n6, B=@n7), B=ScalarParameter() {defaultValue=0.5, parameterName=HitWave-DurationNew}))
  n9 = Multiply(
   A=ScalarParameter() {defaultValue=300, parameterName=HitWave-SizeMultiplierOuterRing}
   B=Power(Base=@n8, Exp=ScalarParameter() {defaultValue=0.5, parameterName=HitWave-OuterRingDelay}) {constExponent=0.7}
  )
  n10 = ScalarParameter() {defaultValue=0.7, parameterName=Hitwave-SphereHard1}
  n11 = ComponentMask(None=VectorParameter() {parameterName=HitWave-Loc, defaultValue=<rgba 100 -1400 -150 0>}.RGB) {b=true, g=true, r=true}
  n12 = WorldPosition() {worldPositionShaderOffset=WPT_ExcludeAllShaderOffsets}
  n13 = LinearInterpolate(A=ScalarParameter() {defaultValue=-0.1, parameterName=HitWave-InnerRingDelay}, Alpha=@n8) {constA=-0.7}
  n14 = SphereMask(
   A=@n12.XYZ
   B=@n11
   Radius=Multiply(A=Saturate(None=@n13), B=ScalarParameter() {defaultValue=1000, parameterName=HitWave-SizeMultiplierInnerRing})
   Hardness=ScalarParameter() {defaultValue=0.25, parameterName=Hitwave-SphereHard2}
  ) {hardnessPercent=0}
  n15 = Add(
   A=SphereMask(
    A=ComponentMask(None=@n2.XYZ) {b=true}
    B=ComponentMask(None=@n5) {b=true}
    Radius=ScalarParameter() {defaultValue=100, parameterName=WPOWipeRadius}
   ) {hardnessPercent=0}
   B=Multiply(A=SphereMask(A=@n12.XYZ, B=@n11, Radius=@n9, Hardness=@n10) {hardnessPercent=0}, B=OneMinus(None=@n14))
  )
  n16 = Constant()
  n17 = FeatureLevelSwitch(Default=@n15, ES2=@n16, ES3_1=@n16, SM4_REMOVED=@n15)
  n18 = Power(Base=Saturate(None=@n17), Exp=ScalarParameter() {defaultValue=5, parameterName=Hitwave-pow})
  n19 = Multiply(A=Multiply(A=Saturate(None=@n18), B=ScalarParameter() {defaultValue=16, parameterName=WPO-AddWave2}), B=VertexNormalWS())
  FeatureLevelSwitch(Default=@n19, ES2=@n1, ES3_1=@n1, SM4_REMOVED=@n19)
  MaterialFunctionCall() {materialFunction=&DebugTimeSine}
  n20 = ScalarParameter() {defaultValue=0.25, parameterName=Hitwave-Brightness}
  n21 = OneMinus(None=@n8)
  n22 = Add(
   A=Multiply(A=VectorParameter() {parameterName=HitWave-Tint, defaultValue=<rgba 0.4034 0.7537 1 0>}.RGB, B=@n17)
   B=Multiply(A=Power(Base=@n21) {constExponent=4}, B=ScalarParameter() {defaultValue=0.25, parameterName=Hitwave-AddBrighness})
  )
  n23 = Time()
  n24 = ScalarParameter() {defaultValue=0.05, parameterName=DistortA-Speed}
  n25 = Add(
   A=Multiply(A=TextureCoordinate(), B=ScalarParameter() {defaultValue=1, parameterName=DistortA-Tile})
   B=Multiply(A=Multiply(A=@n23, B=@n24), B=Constant2Vector() {g=0.25, r=1})
  )
  n26 = TextureCoordinate()
  n27 = ScalarParameter() {defaultValue=2, parameterName=DistortB-Tile}
  n28 = Multiply(A=Time(), B=ScalarParameter() {defaultValue=-0.1, parameterName=DistortB-Speed})
  n29 = Constant2Vector() {g=1, r=0.25}
  n30 = Add(A=Constant() {r=0.33}, B=Add(A=Multiply(A=@n26, B=@n27), B=Multiply(A=@n28, B=@n29)))
  n31 = ComponentMask(None=Add(
    A=TextureSample(UVs=@n25) {samplerType=SAMPLERTYPE_LinearColor, texture=/Figment_S09/Environments/Sets/DigSite/Textures/T_Push_Flow}.RGB
    B=TextureSample(UVs=@n30) {samplerType=SAMPLERTYPE_LinearColor, texture=&T_Push_Flow}.RGB
   )) {g=true, r=true}
  n32 = Multiply(A=ScalarParameter() {defaultValue=35, parameterName=Distort-BaseAmt}, B=Subtract(A=Normalize(VectorInput=@n31)) {constB=0.5})
  n33 = Multiply(A=@n17, B=ScalarParameter() {defaultValue=5, parameterName=Hitwave-DistortAmt3})
  n34 = Add(A=@n32, B=Multiply(A=@n33, B=ScalarParameter() {defaultValue=0.75, parameterName=Distort-RedAmt}))
  n35 = Multiply(A=TextureCoordinate(), B=ScalarParameter() {defaultValue=30, parameterName=LinesTile})
  n36 = Multiply(A=Add(A=@n33, B=@n32), B=ScalarParameter() {defaultValue=1.02, parameterName=Distort-GreenAmt})
  n37 = Add(A=@n33, B=@n32)
  n38 = ScalarParameter() {defaultValue=1.04, parameterName=Distort-BlueAmt}
  n39 = TextureSample(UVs=Add(A=Multiply(A=@n37, B=@n38), B=@n35)) {samplerType=SAMPLERTYPE_LinearColor, texture=/Game/Weapons/WeaponSkins/Textures/T_VertGradient}
  n40 = MaterialFunctionCall(
   X=TextureSample(UVs=Add(A=@n34, B=@n35)) {samplerType=SAMPLERTYPE_LinearColor, texture=&T_VertGradient}.R
   Y=TextureSample(UVs=Add(A=@n36, B=@n35)) {samplerType=SAMPLERTYPE_LinearColor, texture=&T_VertGradient}.R
   Z=@n39.R
  ) {materialFunction=&MakeFloat3}
  n41 = VertexNormalWS()
  n42 = ConstantBiasScale(None=ComponentMask(None=Transform(None=@n41) {transformSourceType=TRANSFORMSOURCE_World, transformType=TRANSFORM_View}) {r=true})
  n43 = Constant() {r=1.5}
  n44 = Saturate(None=Multiply(A=Power(Base=@n42, Exp=@n43), B=Constant() {r=8}))
  n45 = OneMinus(None=@n42)
  n46 = Saturate(None=Add(A=Multiply(A=@n44, B=@n45), B=ScalarParameter() {defaultValue=0.55, parameterName=RainbowMaskAmt01}))
  n47 = Multiply(A=Time(), B=ScalarParameter() {defaultValue=-0.5, parameterName=RainbowSpeed})
  n48 = ComponentMask(None=Transform(None=Constant3Vector() {constant=<rgba 0 0 1 0>}) {transformType=TRANSFORM_View}) {g=true, r=true}
  n49 = ConstantBiasScale(None=Multiply(A=Add(A=@n47, B=@n48), B=Constant2Vector() {g=-1, r=1}) {constB=0.5})
  n50 = ScalarParameter() {defaultValue=0.5, parameterName=RainbowScale}
  n51 = Multiply(
   A=TextureSample(UVs=MaterialFunctionCall(UVs=@n49, Texture Scale=@n50) {materialFunction=&ScaleUVsByCenter}.UVs) {texture=/Game/Athena/Events/NewYear/DiscoNY_Gradient}.RGB
   B=@n39.R
  )
  n52 = Power(
   Base=QualitySwitch(Default=LinearInterpolate(A=@n51, B=@n40.Result, Alpha=@n46), Low=@n40.Result)
   Exp=ScalarParameter() {defaultValue=3, parameterName=LinesPower}
  )
  n53 = Saturate(None=DotProduct(A=VertexNormalWS(), B=CameraVectorWS()))
  n54 = Multiply(A=@n51, B=ScalarParameter() {defaultValue=0.05, parameterName=InbetweenRainbowAmt})
  n55 = TextureSample(UVs=@n25) {samplerType=SAMPLERTYPE_LinearColor, texture=/Figment_S09/Environments/Sets/DigSite/Textures/T_Push_Mask}
  n56 = Power(Base=OneMinus(None=@n55.R), Exp=ScalarParameter() {defaultValue=5, parameterName=InbetweenRainbowMaskPow})
  n57 = Constant()
  n58 = ScalarParameter() {defaultValue=1, parameterName=BubbleDistantDarkAmt}
  n59 = MaterialFunctionCall!(
   Blend Range=ScalarParameter() {defaultValue=40000, parameterName=InnerFresRange}
   Start Offset=ScalarParameter() {defaultValue=-1000, parameterName=InnerFresOffset}
  ) {materialFunction=/Engine/Functions/Engine_MaterialFunctions03/Particles/Distance_Blend}
  n60 = Saturate(None=Add(A=@n55.R, B=LinearInterpolate(A=@n57, B=@n58, Alpha=@n59.Result)))
  n61 = Multiply(A=Add(A=Multiply(A=@n54, B=@n56), B=Multiply(A=@n52, B=@n60)), B=ScalarParameter() {defaultValue=0.2, parameterName=LinesBrightAmt})
  n62 = LinearInterpolate(A=Constant() {r=1}, B=ScalarParameter() {defaultValue=2, parameterName=DistantBrightness}, Alpha=@n59.Result)
  n63 = Multiply(A=@n59.Result, B=ScalarParameter() {defaultValue=0.2, parameterName=DistantAdd})
  n64 = OneMinus(None=Saturate(None=ScalarParameter() {defaultValue=1, parameterName=Opacity}))
  n65 = Constant3Vector() {constant=<rgba 0.1567 0.64 1 0>}
  n66 = Add(A=Multiply(A=Multiply(A=@n64, B=@n65), B=@n18), B=Multiply(A=@n18) {constB=0.1})
  n67 = Constant()
  n68 = FeatureLevelSwitch(Default=@n66, ES2=@n67, ES3_1=@n67, SM4_REMOVED=@n66)
  n69 = Saturate(None=Add(A=Multiply(A=@n61, B=@n62), B=Add(A=@n63, B=@n68)))
  n70 = Add(A=Power(Base=@n53, Exp=ScalarParameter() {defaultValue=4, parameterName=InnerFresPow}), B=@n69)
  n71 = Multiply(A=@n53, B=ScalarParameter() {defaultValue=2, parameterName=FresBright})
  n72 = Add(A=Multiply(A=Multiply(A=@n22, B=@n20), B=@n52), B=Multiply(A=Multiply(A=@n70, B=@n69), B=Saturate(None=@n71)))
  StaticSwitchParameter(True=Multiply(A=@n72, B=ParticleColor().RGB), False=@n72) {parameterName=UseParticleColor}
  n73 = ScalarParameter() {defaultValue=1, parameterName=Opacity}
  StaticSwitchParameter(True=DepthFade(Opacity=@n73, FadeDistance=ScalarParameter() {defaultValue=64, parameterName=DepthFade}), False=@n73) {parameterName=UseDepthFade}
  Time() {bOverride_Period=true, period=0.5}
 ]
}
